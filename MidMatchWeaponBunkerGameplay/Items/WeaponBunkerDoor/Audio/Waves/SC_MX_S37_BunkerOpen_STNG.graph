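# Sound cue for the music stinger tied to a weapon bunker door opening.
# A single wave player feeds the processed stinger wave straight to the output at boosted volume, in the Athena SFX sound class rather than a music class.
# Attenuation overrides supersede the referenced bunker alarm attenuation asset, adding long-range falloff, distance-based filtering, occlusion and submix sends.
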SoundCue /MidMatchWeaponBunkerGameplay/Items/WeaponBunkerDoor/Audio/Waves/SC_MX_S37_BunkerOpen_STNG
root=WavePlayer_0
nodes [{
  id=WavePlayer_0
  properties {soundWaveAssetPtr=&/MX_S37_BunkerOpen_STNG_Processed}
 }]
properties {
 attenuationSettings=/MidMatchWeaponBunkerGameplay/Items/WeaponBunkerDoor/MMOBunkers_Ch6S4/Sound/Waves/ATT_TidalCrane_Bunker_Alarm
 bOverrideAttenuation=true
 maxDistance=15000
 soundClassObject=/Game/Sounds/FortSoundClassesAndModes/Fort_Events_SoundClass/Fort_Athena_SFX
 volumeMultiplier=2
 attenuationOverrides {
  bAttenuateWithLPF=true
  bEnableLogFrequencyScaling=true
  bEnableOcclusion=true
  bEnableSubmixSends=true
  falloffDistance=10000
  hPFFrequencyAtMax=70
  hPFFrequencyAtMin=100
  lPFFrequencyAtMax=1500
  lPFRadiusMax=10000
  lPFRadiusMin=10000
  occlusionLowPassFilterFrequency=10000
  attenuationShapeExtents {x=5000}
 }
}
